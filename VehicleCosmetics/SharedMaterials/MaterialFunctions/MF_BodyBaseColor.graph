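# Material function that builds vehicle body base color by compositing team, accent, paint and trim colors over a diffuse texture.
# Skin alpha blends TeamColor to AccentColor, then skin green blends in PaintColor.
# A biased, scaled and clamped skin red channel lays that result over the diffuse RGB.
# CurvaturePack green applies TrimColor last, producing the BaseColor output.
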MaterialFunction /VehicleCosmetics/SharedMaterials/MaterialFunctions/MF_BodyBaseColor
graph {nodes [
  n1 = ComponentMask(None=FunctionInput() {inputName=Diffuse, inputType=FunctionInput_Vector4, sortPriority=6}) {b=true, g=true, r=true}
  n2 = FunctionInput() {inputName=Skin, inputType=FunctionInput_Vector4, sortPriority=5}
  n3 = LinearInterpolate(
   A=LinearInterpolate(
    A=FunctionInput() {inputName=TeamColor, sortPriority=1}
    B=FunctionInput() {inputName=AccentColor, sortPriority=2}
    Alpha=ComponentMask(None=@n2) {a=true}
   ) {desc=Skin_AlphaMask}
   B=FunctionInput() {inputName=PaintColor, sortPriority=3}
   Alpha=ComponentMask(None=@n2) {g=true}
  ) {desc=PaintColor G Mask}
  n4 = Clamp(None=ConstantBiasScale(None=ComponentMask(None=@n2) {r=true}) {bias=-0.1, scale=2})
  n5 = FunctionInput() {inputName=CurvaturePack, inputType=FunctionInput_Vector4, sortPriority=5}
  FunctionOutput(None=LinearInterpolate(
    A=LinearInterpolate(A=@n1, B=@n3, Alpha=@n4) {desc=DiffuseTexture}
    B=FunctionInput() {inputName=TrimColor, sortPriority=4}
    Alpha=ComponentMask(None=@n5) {g=true}
   ) {desc=TrimColorMask G channel curvature pack}) {outputName=BaseColor}
 ]}
