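# MF_RGBA_Splitter is a utility material function for shared vehicle cosmetics materials that splits a four-channel vector into separate scalar channels.
# A single Vector4 input named RGBA feeds four ComponentMask nodes, each isolating one channel.
# The masks drive four sorted outputs named R, G, B and A.
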MaterialFunction /VehicleCosmetics/SharedMaterials/MaterialFunctions/MF_RGBA_Splitter
graph {nodes [
  n1 = FunctionInput() {inputName=RGBA, inputType=FunctionInput_Vector4, sortPriority=5}
  FunctionOutput(None=ComponentMask(None=@n1) {r=true}) {outputName=R, sortPriority=1}
  FunctionOutput(None=ComponentMask(None=@n1) {g=true}) {outputName=G, sortPriority=2}
  FunctionOutput(None=ComponentMask(None=@n1) {b=true}) {outputName=B, sortPriority=3}
  FunctionOutput(None=ComponentMask(None=@n1) {a=true}) {outputName=A, sortPriority=4}
 ]}
