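# Material function that computes a vehicle body metallic value from two four-channel inputs, Metallic and Skin.
# Clamped Metallic channels are blended by Skin masks: Skin G blends R toward B, then biased, scaled Skin R blends A toward that result.
# The Skin A blend has no effect because both its ends read Metallic R.
# The output is clamped to avoid fully metallic or fully non-metallic values.
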MaterialFunction /VehicleCosmetics/SharedMaterials/MaterialFunctions/MF_BodyMetallic
graph {nodes [
  n1 = MaterialFunctionCall!(RGBA=Clamp(None=FunctionInput() {inputName=Metallic, inputType=FunctionInput_Vector4, sortPriority=7}) {desc=Body_Metallic_Clamp, maxDefault=0.75}) {materialFunction=&/MF_RGBA_Splitter}
  n2 = MaterialFunctionCall(RGBA=FunctionInput() {inputName=Skin, inputType=FunctionInput_Vector4, sortPriority=5}) {materialFunction=&MF_RGBA_Splitter}
  n3 = LinearInterpolate(A=LinearInterpolate(A=@n1.R, B=@n1.R, Alpha=@n2.A), B=@n1.B, Alpha=@n2.G)
  n4 = Clamp(None=ConstantBiasScale(None=@n2.R) {bias=-0.1, scale=2})
  FunctionOutput(None=Clamp(None=LinearInterpolate(A=@n1.A, B=@n3, Alpha=@n4)) {maxDefault=0.85, minDefault=0.15})
 ]}
