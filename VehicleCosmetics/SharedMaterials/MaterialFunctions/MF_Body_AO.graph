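# MF_Body_AO is a material function in the vehicle cosmetics shared materials that darkens a body base color with ambient occlusion.
# It takes the AO value from the alpha channel of the Diffuse input and clamps it with a floor of one half to limit the darkening.
# The clamped value multiplies the BaseColor input to produce the BaseColor with AO output.
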MaterialFunction /VehicleCosmetics/SharedMaterials/MaterialFunctions/MF_Body_AO
graph {nodes [
  n1 = ComponentMask(None=FunctionInput() {inputName=Diffuse, inputType=FunctionInput_Vector4, sortPriority=6}) {a=true}
  FunctionOutput(None=Multiply(A=FunctionInput() {inputName=BaseColor}, B=Clamp(None=@n1) {minDefault=0.5})) {outputName=BaseColor with AO}
 ]}
