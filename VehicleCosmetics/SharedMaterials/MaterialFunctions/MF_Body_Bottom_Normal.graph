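# Material function for vehicle cosmetic materials that sets the strength of a detail normal, by its name for the bottom of the vehicle body.
# It linearly interpolates from a flat normal to the DetailNormal input, using the scalar Bottom_Normal_Alpha input as the blend factor.
# The blended vector goes straight to the output without renormalization.
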MaterialFunction /VehicleCosmetics/SharedMaterials/MaterialFunctions/MF_Body_Bottom_Normal
graph {nodes [FunctionOutput(None=LinearInterpolate(
    A=Constant3Vector() {constant=<rgba 0 0 1 0>}
    B=FunctionInput() {inputName=DetailNormal, sortPriority=12, previewValue=<0 0 0 0>}
    Alpha=FunctionInput() {inputName=Bottom_Normal_Alpha, inputType=FunctionInput_Scalar}
   ))]}
