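# Material function for vehicle body cosmetics that derives clear coat roughness and thickness from a skin mask.
# Each four-channel parameter input holds values for four mask regions: skin alpha blends X to Y, then skin green blends toward Z.
# A biased, scaled and clamped skin red channel blends from W to that result, so low red yields W.
# Roughness and thickness share the same blend chain and masks.
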MaterialFunction /VehicleCosmetics/SharedMaterials/MaterialFunctions/MF_Body_ClearCoat
graph {nodes [
  n1 = FunctionInput() {inputName=ClearCoat Roughness, inputType=FunctionInput_Vector4, sortPriority=10}
  n2 = MaterialFunctionCall!(RGBA=FunctionInput() {inputName=Skin, inputType=FunctionInput_Vector4, sortPriority=5}) {materialFunction=&/MF_RGBA_Splitter}
  n3 = LinearInterpolate(A=ComponentMask(None=@n1) {r=true}, B=ComponentMask(None=@n1) {g=true}, Alpha=@n2.A)
  n4 = ComponentMask(None=@n1) {b=true}
  n5 = Clamp(None=ConstantBiasScale(None=@n2.R) {bias=-0.1, scale=2})
  FunctionOutput(None=LinearInterpolate(A=ComponentMask(None=@n1) {a=true}, B=LinearInterpolate(A=@n3, B=@n4, Alpha=@n2.G), Alpha=@n5)) {outputName=ClearCoat_Roughness}
  n6 = FunctionInput() {inputName=ClearCoat Thickness, inputType=FunctionInput_Vector4, sortPriority=9}
  n7 = LinearInterpolate(A=ComponentMask(None=@n6) {r=true}, B=ComponentMask(None=@n6) {g=true}, Alpha=@n2.A)
  n8 = ComponentMask(None=@n6) {b=true}
  FunctionOutput(None=LinearInterpolate(A=ComponentMask(None=@n6) {a=true}, B=LinearInterpolate(A=@n7, B=@n8, Alpha=@n2.G), Alpha=@n5)) {outputName=ClearCoat_Thickness}
 ]}
