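# Material function for vehicle body materials that applies detail-normal AO to base color and outputs a detail normal.
# DetailNormal_Alpha forms an AO multiplier on BaseColor, with a further blend labeled to exclude trim.
# The normal blends from flat to DetailNormal by a clamped value minus CurvaturePack green.
# DetailNormalControl channels feed team, accent, paint, trim and window blends, though this chain reaches no output and several inputs appear unconnected.
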MaterialFunction /VehicleCosmetics/SharedMaterials/MaterialFunctions/MF_Body_DetailNormal
graph {nodes [
  FunctionInput() {inputName=Skin, inputType=FunctionInput_Vector4}
  n1 = FunctionInput() {inputName=BaseColor}
  n2 = FunctionInput() {inputName=DetailNormal_Alpha, inputType=FunctionInput_Scalar, sortPriority=2, previewValue=<0 0 0 0>}
  n3 = LinearInterpolate(B=Clamp(None=@n2) {minDefault=0.03}) {constA=1, desc=Detail AO Masks}
  FunctionOutput(None=LinearInterpolate(A=Multiply(A=@n1, B=@n3), B=@n1) {desc=No detail normal on trim}) {outputName=BaseColor + DetailNormal AO}
  n4 = Subtract(A=Reroute(), B=ComponentMask(None=FunctionInput() {inputName=CurvaturePack, inputType=FunctionInput_Vector4, sortPriority=1}) {g=true})
  FunctionOutput(None=LinearInterpolate(
    A=Constant3Vector() {constant=<rgba 0 0 1 0>}
    B=FunctionInput() {inputName=DetailNormal, sortPriority=1, previewValue=<0 0 0 0>}
    Alpha=Clamp(None=@n4)
   )) {outputName=BottomNormal, sortPriority=2}
  n5 = FunctionInput() {inputName=DetailNormalControl, inputType=FunctionInput_Vector4, sortPriority=3, previewValue=<0 0 0 0>}
  n6 = LinearInterpolate(A=ComponentMask(None=@n5) {r=true}, B=ComponentMask(None=@n5) {g=true}) {desc=Team and Accent}
  n7 = ComponentMask(None=@n5) {b=true}
  n8 = ConstantBiasScale() {bias=-0.1, scale=2}
  LinearInterpolate(A=LinearInterpolate(A=ComponentMask(None=@n5) {a=true}, B=LinearInterpolate(A=@n6, B=@n7) {desc=Paint}, Alpha=Clamp(None=@n8)) {desc=Trim}) {constB=0, desc=Window}
  MaterialFunctionCall()
 ]}
