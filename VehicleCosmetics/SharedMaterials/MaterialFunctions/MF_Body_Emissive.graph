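# Material function for vehicle body emissive, turning a packed EmissiveColor mask and a Skin mask into an emissive output tinted by BaseColor.
# Skin green blends EmissiveColor red toward blue; an earlier lerp by Skin alpha is a no-op, since both inputs are red.
# Skin red, biased, scaled and clamped, blends from EmissiveColor alpha to that result, which then multiplies BaseColor.
# EmissiveColor green is unused.
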MaterialFunction /VehicleCosmetics/SharedMaterials/MaterialFunctions/MF_Body_Emissive
graph {nodes [
  n1 = FunctionInput() {inputName=EmissiveColor, inputType=FunctionInput_Vector4, sortPriority=11, previewValue=<0 0 0 0>}
  n2 = MaterialFunctionCall!(RGBA=@n1) {materialFunction=&/MF_RGBA_Splitter}
  n3 = MaterialFunctionCall(RGBA=FunctionInput() {inputName=Skin, inputType=FunctionInput_Vector4, sortPriority=5}) {materialFunction=&MF_RGBA_Splitter}
  n4 = LinearInterpolate(A=LinearInterpolate(A=@n2.R, B=@n2.R, Alpha=@n3.A), B=@n2.B, Alpha=@n3.G)
  n5 = Clamp(None=ConstantBiasScale(None=@n3.R) {bias=-0.1, scale=2})
  FunctionOutput(None=Multiply(A=LinearInterpolate(A=@n2.A, B=@n4, Alpha=@n5), B=FunctionInput() {inputName=BaseColor}) {desc=Multiply by color})
 ]}
