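# Material function that adds a view-dependent iridescent sheen to a vehicle body's base color.
# A Fresnel term, with exponent from Iridescent Coverage, combines with the Body_FX red channel to drive a hue shift of TeamColor.
# The same Fresnel blends this shifted color over BaseColor.
# The CurvaturePack green channel, noted as the trim color mask, restores the original BaseColor on trim.
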MaterialFunction /VehicleCosmetics/SharedMaterials/MaterialFunctions/MF_Body_Iridescent
graph {nodes [
  n1 = FunctionInput() {inputName=BaseColor, sortPriority=1}
  n2 = ComponentMask(None=FunctionInput() {inputName=Body_FX, inputType=FunctionInput_Vector4, sortPriority=2}) {r=true}
  n3 = Fresnel(ExponentIn=FunctionInput() {inputName=Iridescent Coverage, inputType=FunctionInput_Scalar, sortPriority=3}) {exponent=3}
  n4 = Multiply(B=Multiply(A=Add(A=@n2, B=@n3), B=@n3)) {constA=3}
  n5 = FunctionInput() {inputName=TeamColor}
  n6 = ComponentMask(None=AppendVector(
    A=MaterialFunctionCall(Hue Shift Percentage=@n4, Texture=@n5) {materialFunction=&HueShift}.Result
    B=Constant() {r=1}
   )) {b=true, g=true, r=true}
  n7 = Multiply(A=@n3)
  n8 = FunctionInput() {inputName=CurvaturePack, inputType=FunctionInput_Vector4, sortPriority=5}
  FunctionOutput(None=LinearInterpolate(A=LinearInterpolate(A=@n1, B=@n6, Alpha=@n7), B=@n1, Alpha=ComponentMask(None=@n8) {g=true}) {desc=TrimColorMask G channel curvature pack}) {outputName=Basecolor with Iridescent}
 ]}
