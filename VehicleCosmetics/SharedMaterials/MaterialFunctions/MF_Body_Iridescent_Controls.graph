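# Material function for vehicle body materials that adds a view-dependent iridescent tint to base color in selected paint regions.
# A coverage-driven Fresnel, offset by Body_FX red, hue-shifts TeamColor and blends it over BaseColor.
# Skin channels build team, accent, paint and trim masks, each scaled by an Iridescent_Control channel in chained lerps.
# The accent mask subtracts Skin alpha from itself, so that layer adds nothing.
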MaterialFunction /VehicleCosmetics/SharedMaterials/MaterialFunctions/MF_Body_Iridescent_Controls
graph {nodes [
  n1 = FunctionInput() {inputName=BaseColor}
  n2 = Fresnel(ExponentIn=FunctionInput() {inputName=Iridescent Coverage, inputType=FunctionInput_Scalar, sortPriority=11}) {exponent=3}
  n3 = Add(A=ComponentMask(None=FunctionInput() {inputName=Body_FX, inputType=FunctionInput_Vector4, sortPriority=2}) {r=true}, B=@n2)
  n4 = MaterialFunctionCall(Hue Shift Percentage=Multiply(B=Multiply(A=@n3, B=@n2)) {constA=3}, Texture=FunctionInput() {inputName=TeamColor}) {materialFunction=&HueShift}
  n5 = Constant() {r=1}
  n6 = LinearInterpolate(A=@n1, B=ComponentMask(None=AppendVector(A=@n4.Result, B=@n5)) {b=true, g=true, r=true}, Alpha=Multiply(A=@n2))
  n7 = MaterialFunctionCall!(RGBA=FunctionInput() {inputName=Iridescent_Control, inputType=FunctionInput_Vector4, sortPriority=7}) {materialFunction=&/MF_RGBA_Splitter}
  n8 = MaterialFunctionCall(RGBA=FunctionInput() {inputName=Skin, inputType=FunctionInput_Vector4, sortPriority=5}) {materialFunction=&MF_RGBA_Splitter}
  n9 = Clamp(None=ConstantBiasScale(None=@n8.R) {bias=-0.1, scale=2})
  n10 = Subtract(A=@n9, B=@n8.A)
  n11 = Subtract(A=@n8.A, B=@n8.A)
  n12 = LinearInterpolate(A=LinearInterpolate(A=@n1, B=@n6, Alpha=Multiply(A=@n7.R, B=@n10)) {desc="Team "}, B=@n6, Alpha=Multiply(A=@n7.G, B=Clamp(None=@n11))) {desc=Accent}
  n13 = Multiply(A=@n7.B, B=@n8.G)
  n14 = OneMinus(None=@n9)
  FunctionOutput(None=LinearInterpolate(A=LinearInterpolate(A=@n12, B=@n6, Alpha=@n13) {desc=Paint}, B=@n6, Alpha=Multiply(A=@n14, B=@n7.A)) {constB=0, desc=Trim}) {outputName=Basecolor with Iridescent}
 ]}
