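# Material function for vehicle body cosmetics that adds a view-dependent pearlescent tint by blending BaseColor toward AccentColor.
# Pearlescent_Coverage sets the exponent of a biased, scaled Fresnel and also multiplies its result.
# Pearlescent_Control channels weight that term per region, while Skin mask channels select between team, accent, paint and trim through nested lerps.
# Window areas are forced to zero, and the clamped result drives the final color blend.
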MaterialFunction /VehicleCosmetics/SharedMaterials/MaterialFunctions/MF_Body_Pearlescent
graph {nodes [
  n1 = FunctionInput() {inputName=Pearlescent_Coverage, inputType=FunctionInput_Scalar, sortPriority=8}
  n2 = Multiply(A=ConstantBiasScale(None=Fresnel(ExponentIn=@n1) {exponent=6}) {bias=-0.2, scale=2}, B=@n1)
  n3 = MaterialFunctionCall!(RGBA=FunctionInput() {inputName=Pearlescent_Control, inputType=FunctionInput_Vector4, sortPriority=7}) {materialFunction=&/MF_RGBA_Splitter}
  n4 = MaterialFunctionCall(RGBA=FunctionInput() {inputName=Skin, inputType=FunctionInput_Vector4, sortPriority=5}) {materialFunction=&MF_RGBA_Splitter}
  n5 = LinearInterpolate(A=Multiply(A=@n2, B=@n3.R), B=Multiply(A=@n2, B=@n3.G), Alpha=@n4.A) {desc=Team and Accent}
  n6 = Multiply(A=@n2, B=@n3.B)
  n7 = LinearInterpolate(
   A=LinearInterpolate(A=Multiply(A=@n2, B=@n3.A), B=LinearInterpolate(A=@n5, B=@n6, Alpha=@n4.G) {desc=Paint}, Alpha=@n4.R) {desc=Trim}
   Alpha=@n4.B
  ) {constB=0, desc=Window}
  FunctionOutput(None=LinearInterpolate(A=FunctionInput() {inputName=BaseColor}, B=FunctionInput() {inputName=AccentColor, sortPriority=2}, Alpha=Clamp(None=@n7))) {outputName=BaseColor with Pearlescent}
 ]}
