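# Material function that computes a blended roughness value for vehicle body materials from a Roughness vector and a Skin mask vector.
# Both inputs are split into channels: Skin G blends Roughness R toward B, then a biased, scaled and clamped Skin R blends Roughness A toward that result.
# The Skin A blend has no effect because both its ends are Roughness R, and Roughness G is unused.
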MaterialFunction /VehicleCosmetics/SharedMaterials/MaterialFunctions/MF_Body_Roughness
graph {nodes [
  n1 = MaterialFunctionCall!(RGBA=FunctionInput() {inputName=Roughness, inputType=FunctionInput_Vector4, sortPriority=8}) {materialFunction=&/MF_RGBA_Splitter}
  n2 = MaterialFunctionCall(RGBA=FunctionInput() {inputName=Skin, inputType=FunctionInput_Vector4, sortPriority=5}) {materialFunction=&MF_RGBA_Splitter}
  n3 = LinearInterpolate(A=@n1.R, B=@n1.R, Alpha=@n2.A)
  n4 = ConstantBiasScale(None=@n2.R) {bias=-0.1, scale=2}
  FunctionOutput(None=LinearInterpolate(A=@n1.A, B=LinearInterpolate(A=@n3, B=@n1.B, Alpha=@n2.G), Alpha=Clamp(None=@n4)))
 ]}
