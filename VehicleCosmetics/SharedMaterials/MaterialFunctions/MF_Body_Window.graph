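# Material function that paints an opaque-looking dark glass window finish onto vehicle body materials wherever the scalar BMask is set.
# It breaks the incoming material attributes and blends base color toward a near-black Fresnel gradient, metallic toward zero and roughness toward a glossy low value.
# Emissive, normal, clear coat, clear coat roughness and ambient occlusion pass through into a newly built attribute set, so other incoming attributes are not carried over.
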MaterialFunction /VehicleCosmetics/SharedMaterials/MaterialFunctions/MF_Body_Window
graph {nodes [
  n1 = BreakMaterialAttributes!(Attr=GetMaterialAttributes(None=FunctionInput() {inputName=Attributes IN, inputType=FunctionInput_MaterialAttributes}).MaterialAttributes)
  n2 = Fresnel()
  n3 = FunctionInput() {inputName=BMask, inputType=FunctionInput_Scalar, sortPriority=1}
  n4 = SetMaterialAttributes!(
   MaterialAttributes
   BaseColor=LinearInterpolate(A=@n1.BaseColor, B=LinearInterpolate(Alpha=@n2) {constB=0.05}, Alpha=@n3) {constB=0}
   Metallic=LinearInterpolate(A=@n1.Metallic, Alpha=@n3) {constB=0}
   Roughness=LinearInterpolate(A=@n1.Roughness, Alpha=@n3) {constB=0.1}
   EmissiveColor=@n1.EmissiveColor
   Normal=@n1.Normal
   CustomData0=@n1.ClearCoat
   CustomData1=@n1.ClearCoatRoughness
   AmbientOcclusion=@n1.AmbientOcclusion
  ) {attributeSetTypes [
    #1
    #2
    #3
    #4
    #5
    #6
    #7
    #8
   ]}
  FunctionOutput(None=@n4)
 ]}
