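# Material function for vehicle cosmetics that outputs four-band audio Spectrum and Amplitude for beat-synced effects, with a faked beat as fallback.
# MF_AudioReactivity supplies the real signal, and a floored, saturated threshold test on each output hard-switches from fake to real once audio is detected.
# Time-driven sawtooths fake bass and low, sine waves fake mid and high, and their maximum forms fake amplitude.
# Separate inputs scale the fake beat and overall intensity.
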MaterialFunction /VehicleCosmetics/SharedMaterials/MaterialFunctions/MF_AudioBeatSync
graph {nodes [
  n1 = NamedRerouteUsage()
  n2 = MaterialFunctionCall() {materialFunction=/Game/Packages/Fortress_SharedMaterials/Base_Material_Functions/BaseCharacterShaderFunctions/Audio/MF_AudioReactivity}
  n3 = Length(None=@n2.Spectrum) {is=class MaterialExpressionLength}
  n4 = Divide(B=ScalarParameter() {defaultValue=0.05, parameterName=FrontendFakeAudio_DetectionThreshold})
  n5 = Saturate(None=Floor(None=Multiply(A=@n3, B=@n4) {constB=20}))
  n6 = FunctionInput(Preview=Constant() {r=1}) {bUsePreviewValueAsDefault=true, description=Overall effect intensity, inputName=BeatSyncIntensity, inputType=FunctionInput_Scalar}
  FunctionOutput(None=Multiply(A=LinearInterpolate(A=@n1, B=@n2.Spectrum, Alpha=@n5), B=@n6)) {desc="RGBA = Bass/Low/Mid/High", outputName=Spectrum}
  n7 = NamedRerouteUsage()
  n8 = Saturate(None=Floor(None=Multiply(A=@n2.Amplitude, B=@n4) {constB=20}))
  FunctionOutput(None=Multiply(A=LinearInterpolate(A=@n7, B=@n2.Amplitude, Alpha=@n8), B=@n6)) {desc=Overall Volume, outputName=Amplitude}
  n9 = FunctionInput(Preview=Constant() {r=1}) {
   bUsePreviewValueAsDefault=true
   description=Intensity of generic faked beat effect
   inputName=FakeBeatIntensity
   inputType=FunctionInput_Scalar
  }
  n10 = Time() {bOverride_Period=true, period=6000}
  n11 = Multiply(A=@n10, B=ScalarParameter() {defaultValue=1.5, parameterName=FakeLowBassSpeed})
  n12 = Constant() {r=1.5}
  n13 = OneMinus(None=Fmod(A=@n11, B=Constant() {r=1}))
  n14 = Add(A=OneMinus(None=Fmod(A=@n11, B=@n12)), B=@n13)
  n15 = NamedRerouteDeclaration(None=Add(A=Multiply(A=@n14) {constB=0.25}) {constB=0.3, desc=bed}) {name=Fake Bass, nodeColor=<rgba 0.0781 0 1 1>}
  n16 = NamedRerouteDeclaration(None=Add(A=Multiply(A=@n13) {constB=0.25}) {constB=0.3, desc=bed}) {name=Fake Low, nodeColor=<rgba 0 0.2957 0.0475 1>}
  n17 = Max(A=@n15, B=@n16)
  n18 = Sine(None=Multiply(A=@n10, B=ScalarParameter() {defaultValue=2, parameterName=FakeMidHighSpeed}))
  n19 = NamedRerouteDeclaration(None=Add(A=Multiply(A=@n18) {constB=0.2}) {constB=0.075, desc=bed}) {name=Fake Mid, nodeColor=<rgba 1 0.9198 0 1>}
  n20 = Abs(None=@n18)
  n21 = NamedRerouteDeclaration(None=Add(A=Multiply(A=@n20) {constB=0.05}) {constB=0.075, desc=bed}) {name=Fake High, nodeColor=<rgba 1 0.3984 0 1>}
  n22 = Max(A=@n19, B=@n21)
  NamedRerouteDeclaration(None=Multiply(A=@n9, B=Max(A=@n17, B=@n22))) {name=FakeAmplitude, nodeColor=<rgba 0.2422 1 0 1>}
  NamedRerouteDeclaration(None=Multiply(
    A=@n9
    B=MaterialFunctionCall(R=@n15, G=@n16, B=@n19, A=@n21) {materialFunction=&AppendMany}.RGBA
   )) {name=FakeSpectrum, nodeColor=<rgba 1 0 0.5625 1>}
 ]}
