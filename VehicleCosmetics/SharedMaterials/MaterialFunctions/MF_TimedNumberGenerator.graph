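# Material function "Timed Number Generator" that outputs a time-driven stepped scalar, with a configurable additive offset.
# Time times ChangeSpeed wraps through Fmod at half of Instances, then a fixed bias and scale plus Ceil turn it into repeating integer steps.
# Interval sets the size of each step and AdditionalOffset shifts the result.
# The bias is constant, so changing Instances alters the top of the range rather than re-centering it.
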MaterialFunction /VehicleCosmetics/SharedMaterials/MaterialFunctions/MF_TimedNumberGenerator
graph {
 properties {description=Timed number generator with addition offset options., userExposedCaption=Timed Number Generator}
 nodes [
  n1 = FunctionInput(Preview=Constant() {r=0.1}) {inputName=Interval, inputType=FunctionInput_Scalar}
  n2 = Multiply(A=Time(), B=FunctionInput(Preview=Constant() {r=0.25}) {inputName=ChangeSpeed, inputType=FunctionInput_Scalar, sortPriority=4})
  n3 = Divide(A=FunctionInput(Preview=Constant() {r=3}) {inputName=Instances, inputType=FunctionInput_Scalar, sortPriority=2})
  n4 = Ceil(None=ConstantBiasScale(None=Fmod(A=@n2, B=@n3)) {bias=-1.5, scale=2})
  n5 = Constant()
  FunctionOutput(None=Add(A=Multiply(A=@n1, B=@n4), B=FunctionInput(Preview=@n5) {inputName=AdditionalOffset, inputType=FunctionInput_Scalar, sortPriority=6}))
 ]
}
